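# Post-copy AnimBlueprint for the M_MED_DriftFable male body cosmetic, adding muscle jiggle to biceps, triceps and pecs.
# It copies the attached parent mesh's pose through the body's Control Rig, then caches it.
# Two spring-based Anim Dynamics chains, tuned separately, process that pose, and bIsInFrontEnd selects between them.
# bPbBlendEnabled switches instantly between the simulated result and the plain rigged pose.
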AnimBlueprint /BRCosmetics/Characters/Player/Male/Medium/Bodies/M_MED_DriftFable/Meshes/M_MED_DriftFable_AnimBP
target_skeleton=&/M_MED_DriftFable_Skeleton
graphs [
 AnimGraph {nodes [
   Output Pose(Result=Use cached pose 'Dynamics_CACHE'()) {layerGroup=None}
   n1 = Control Rig(Source=Copy Pose From Mesh() {bUseAttachedParent=true}) {controlRigAssetReference {blueprintRigClass=&/M_MED_DriftFable_CtrlRig.M_MED_DriftFable_CtrlRig_C}}
   Save cached pose 'ControlRig_CACHE'(Pose=@n1) {cacheName=ControlRig_CACHE}
   n2 = Use cached pose 'ControlRig_CACHE'()
   Save cached pose 'PreDynamics_CACHE'(Pose=Component To Local(ComponentPose=Local To Component(LocalPose=@n2))) {cacheName=PreDynamics_CACHE}
   n3 = Anim Dynamics - Bone: upperarm_bicep_l(ComponentPose=Local To Component(LocalPose=Use cached pose 'PreDynamics_CACHE'()), Alpha=1, ExternalForce="0,0,0") {
    angularSpringConstant=15000
    bLinearSpring=true
    bOverrideLinearDamping=true
    bUseGravityOverride=true
    bUsePlanarLimit=false
    linearDampingOverride=1
    linearSpringConstant=500
    simulationSpace=RootRelative
    boundBone {boneName=upperarm_bicep_l}
    physicsBodyDefinitions [{
      boundBone {boneName=upperarm_bicep_l}
      boxExtents=<38 38 38>
      constraintSetup {angularLimitsMax=<0 0 0>, angularLimitsMin=<0 0 0>, angularTarget=<1 0 0>, linearAxesMax=<0 0 0.5>, linearAxesMin=<0 0 -0.5>}
      localJointOffset=<-2 0 0>
     }]
    relativeSpaceBone {boneName=upperarm_l}
    simSpaceSettings {maxAngularAcceleration=40, maxAngularVelocity=770, simSpaceAngularAlpha=1}
   }
   n4 = Anim Dynamics - Bone: upperarm_bicep_l(ComponentPose=@n3, Alpha=1, ExternalForce="0,0,0") {
    angularDampingOverride=0.9999
    angularSpringConstant=15000
    bAngularSpring=true
    bOverrideAngularDamping=true
    bUseGravityOverride=true
    bUsePlanarLimit=false
    simulationSpace=RootRelative
    boundBone {boneName=upperarm_bicep_l}
    gravityOverride {z=-980}
    physicsBodyDefinitions [{
      boundBone {boneName=upperarm_bicep_l}
      boxExtents=<24 38 24>
      constraintSetup {angularLimitsMax=<1 10 1>, angularLimitsMin=<-1 -10 -1>, angularTarget=<1 0 0>}
      localJointOffset=<-5 0 0>
     }]
    relativeSpaceBone {boneName=upperarm_l}
    simSpaceSettings {maxAngularAcceleration=10, maxAngularVelocity=300, simSpaceAngularAlpha=1}
   }
   n5 = Anim Dynamics - Bone: upperarm_bicep_r(ComponentPose=@n4, Alpha=1, ExternalForce="0,0,0") {
    angularSpringConstant=15000
    bLinearSpring=true
    bOverrideLinearDamping=true
    bUseGravityOverride=true
    bUsePlanarLimit=false
    linearDampingOverride=1
    linearSpringConstant=500
    simulationSpace=RootRelative
    boundBone {boneName=upperarm_bicep_r}
    physicsBodyDefinitions [{
      boundBone {boneName=upperarm_bicep_r}
      boxExtents=<38 38 38>
      constraintSetup {angularLimitsMax=<0 0 0>, angularLimitsMin=<0 0 0>, angularTarget=<1 0 0>, linearAxesMax=<0 0 0.5>, linearAxesMin=<0 0 -0.5>}
      localJointOffset=<2 0 0>
     }]
    relativeSpaceBone {boneName=upperarm_r}
    simSpaceSettings {maxAngularAcceleration=40, maxAngularVelocity=770, simSpaceAngularAlpha=1}
   }
   n6 = Anim Dynamics - Bone: upperarm_bicep_r(ComponentPose=@n5, Alpha=1, ExternalForce="0,0,0") {
    angularDampingOverride=0.9999
    angularSpringConstant=15000
    bAngularSpring=true
    bOverrideAngularDamping=true
    bUseGravityOverride=true
    bUsePlanarLimit=false
    simulationSpace=RootRelative
    boundBone {boneName=upperarm_bicep_r}
    gravityOverride {z=-980}
    physicsBodyDefinitions [{
      boundBone {boneName=upperarm_bicep_r}
      boxExtents=<24 38 24>
      constraintSetup {angularLimitsMax=<1 10 1>, angularLimitsMin=<-1 -10 -1>, angularTarget=<1 0 0>}
      localJointOffset=<-5 0 0>
     }]
    relativeSpaceBone {boneName=upperarm_r}
    simSpaceSettings {maxAngularAcceleration=10, maxAngularVelocity=300, simSpaceAngularAlpha=1}
   }
   n7 = Anim Dynamics - Bone: upperarm_tricep_l(ComponentPose=@n6, Alpha=1, ExternalForce="0,0,0") {
    angularSpringConstant=15000
    bLinearSpring=true
    bOverrideLinearDamping=true
    bUseGravityOverride=true
    bUsePlanarLimit=false
    linearDampingOverride=1
    linearSpringConstant=450
    simulationSpace=RootRelative
    boundBone {boneName=upperarm_tricep_l}
    gravityOverride {z=-100}
    physicsBodyDefinitions [{
      boundBone {boneName=upperarm_tricep_l}
      boxExtents=<48 58 48>
      constraintSetup {angularLimitsMax=<0 0 0>, angularLimitsMin=<0 0 0>, angularTarget=<1 0 0>, linearAxesMax=<0 0.1 0.75>, linearAxesMin=<0 -0.1 -0.75>}
      localJointOffset=<-3 0 0>
     }]
    relativeSpaceBone {boneName=upperarm_l}
    simSpaceSettings {maxAngularAcceleration=30, maxAngularVelocity=300, simSpaceAngularAlpha=1}
   }
   n8 = Anim Dynamics - Bone: upperarm_tricep_l(ComponentPose=@n7, Alpha=1, ExternalForce="0,0,0") {
    angularDampingOverride=0.9999
    angularSpringConstant=50000
    bAngularSpring=true
    bOverrideAngularDamping=true
    bUseGravityOverride=true
    bUsePlanarLimit=false
    simulationSpace=BoneRelative
    boundBone {boneName=upperarm_tricep_l}
    gravityOverride {z=-980}
    physicsBodyDefinitions [{
      boundBone {boneName=upperarm_tricep_l}
      boxExtents=<24 38 24>
      constraintSetup {angularLimitsMax=<1 10 1>, angularLimitsMin=<-1 -10 -1>, angularTarget=<1 0 0>}
      localJointOffset=<5 0 0>
     }]
    relativeSpaceBone {boneName=upperarm_l}
    simSpaceSettings {maxAngularAcceleration=10, maxAngularVelocity=300, simSpaceAngularAlpha=1}
   }
   n9 = Anim Dynamics - Bone: upperarm_tricep_r(ComponentPose=@n8, Alpha=1, ExternalForce="0,0,0") {
    angularSpringConstant=15000
    bLinearSpring=true
    bOverrideLinearDamping=true
    bUseGravityOverride=true
    bUsePlanarLimit=false
    linearDampingOverride=1
    linearSpringConstant=450
    simulationSpace=RootRelative
    boundBone {boneName=upperarm_tricep_r}
    gravityOverride {z=-100}
    physicsBodyDefinitions [{
      boundBone {boneName=upperarm_tricep_r}
      boxExtents=<48 58 48>
      constraintSetup {angularLimitsMax=<0 0 0>, angularLimitsMin=<0 0 0>, angularTarget=<1 0 0>, linearAxesMax=<0 0.1 0.75>, linearAxesMin=<0 -0.1 -0.75>}
      localJointOffset=<3 0 0>
     }]
    relativeSpaceBone {boneName=upperarm_r}
    simSpaceSettings {maxAngularAcceleration=30, maxAngularVelocity=300, simSpaceAngularAlpha=1}
   }
   n10 = Anim Dynamics - Bone: upperarm_tricep_r(ComponentPose=@n9, Alpha=1, ExternalForce="0,0,0") {
    angularDampingOverride=0.9999
    angularSpringConstant=50000
    bAngularSpring=true
    bOverrideAngularDamping=true
    bUseGravityOverride=true
    bUsePlanarLimit=false
    simulationSpace=BoneRelative
    boundBone {boneName=upperarm_tricep_r}
    gravityOverride {z=-980}
    physicsBodyDefinitions [{
      boundBone {boneName=upperarm_tricep_r}
      boxExtents=<24 38 24>
      constraintSetup {angularLimitsMax=<1 10 1>, angularLimitsMin=<-1 -10 -1>, angularTarget=<1 0 0>}
      localJointOffset=<5 0 0>
     }]
    relativeSpaceBone {boneName=upperarm_r}
    simSpaceSettings {maxAngularAcceleration=10, maxAngularVelocity=300, simSpaceAngularAlpha=1}
   }
   n11 = Anim Dynamics - Bone: deform_pec_l(ComponentPose=@n10, Alpha=1, ExternalForce="0,0,0") {
    angularSpringConstant=15000
    bLinearSpring=true
    bOverrideLinearDamping=true
    bUseGravityOverride=true
    bUsePlanarLimit=false
    linearDampingOverride=1
    linearSpringConstant=400
    simulationSpace=RootRelative
    boundBone {boneName=deform_pec_l}
    gravityOverride {z=-200}
    physicsBodyDefinitions [{
      boundBone {boneName=deform_pec_l}
      boxExtents=<50 55 60>
      constraintSetup {angularLimitsMax=<0 0 0>, angularLimitsMin=<0 0 0>, angularTarget=<1 0 0>, linearAxesMax=<0.5 0 1.5>, linearAxesMin=<-0.5 0 -1.5>}
      localJointOffset=<-4 0 0>
     }]
    relativeSpaceBone {boneName=spine_04}
    simSpaceSettings {maxAngularAcceleration=10, maxAngularVelocity=770, simSpaceAngularAlpha=1}
   }
   n12 = Anim Dynamics - Bone: deform_pec_r(ComponentPose=@n11, Alpha=1, ExternalForce="0,0,0") {
    angularSpringConstant=15000
    bLinearSpring=true
    bOverrideLinearDamping=true
    bUseGravityOverride=true
    bUsePlanarLimit=false
    linearDampingOverride=1
    linearSpringConstant=400
    simulationSpace=RootRelative
    boundBone {boneName=deform_pec_r}
    gravityOverride {z=-200}
    physicsBodyDefinitions [{
      boundBone {boneName=deform_pec_r}
      boxExtents=<50 55 60>
      constraintSetup {angularLimitsMax=<0 0 0>, angularLimitsMin=<0 0 0>, angularTarget=<1 0 0>, linearAxesMax=<0.5 0 1.5>, linearAxesMin=<-0.5 0 -1.5>}
      localJointOffset=<4 0 0>
     }]
    relativeSpaceBone {boneName=spine_04}
    simSpaceSettings {maxAngularAcceleration=10, maxAngularVelocity=770, simSpaceAngularAlpha=1}
   }
   n13 = Component To Local(ComponentPose=@n12)
   n14 = Anim Dynamics - Bone: upperarm_bicep_l!(ComponentPose=Local To Component(LocalPose=Use cached pose 'PreDynamics_CACHE'()), ExternalForce="0,0,0") -> (Pose) {
    alpha=0.8
    angularSpringConstant=15000
    bLinearSpring=true
    bOverrideLinearDamping=true
    bUseGravityOverride=true
    bUsePlanarLimit=false
    linearDampingOverride=0.9999
    linearSpringConstant=500
    simulationSpace=RootRelative
    boundBone {boneName=upperarm_bicep_l}
    componentAppliedLinearAccClamp=<1000 1000 1000>
    physicsBodyDefinitions [{
      boundBone {boneName=upperarm_bicep_l}
      boxExtents=<38 38 38>
      constraintSetup {angularLimitsMax=<0 0 0>, angularLimitsMin=<0 0 0>, angularTarget=<1 0 0>, linearAxesMax=<0 0 1>, linearAxesMin=<0 0 -1>}
      localJointOffset=<-2 0 0>
     }]
    relativeSpaceBone {boneName=pelvis}
    simSpaceSettings {maxAngularAcceleration=40, maxAngularVelocity=770, simSpaceAngularAlpha=0.4}
   }
   n15 = Anim Dynamics - Bone: upperarm_bicep_l(ComponentPose=@n14, ExternalForce="0,0,0") {
    alpha=0.8
    angularDampingOverride=0.9999
    angularSpringConstant=15000
    bAngularSpring=true
    bOverrideAngularDamping=true
    bUseGravityOverride=true
    bUsePlanarLimit=false
    simulationSpace=RootRelative
    boundBone {boneName=upperarm_bicep_l}
    componentAppliedLinearAccClamp=<1000 1000 1000>
    gravityOverride {z=-980}
    physicsBodyDefinitions [{
      boundBone {boneName=upperarm_bicep_l}
      boxExtents=<24 38 24>
      constraintSetup {angularLimitsMax=<1 10 1>, angularLimitsMin=<-1 -10 -1>, angularTarget=<1 0 0>}
      localJointOffset=<-5 0 0>
     }]
    relativeSpaceBone {boneName=pelvis}
    simSpaceSettings {maxAngularAcceleration=10, maxAngularVelocity=300, simSpaceAngularAlpha=0.4}
   }
   n16 = Anim Dynamics - Bone: upperarm_bicep_r!(ComponentPose=@n15, ExternalForce="0,0,0") -> (Pose) {
    alpha=0.8
    angularSpringConstant=15000
    bLinearSpring=true
    bOverrideLinearDamping=true
    bUseGravityOverride=true
    bUsePlanarLimit=false
    linearDampingOverride=0.9999
    linearSpringConstant=500
    simulationSpace=RootRelative
    boundBone {boneName=upperarm_bicep_r}
    componentAppliedLinearAccClamp=<1000 1000 1000>
    physicsBodyDefinitions [{
      boundBone {boneName=upperarm_bicep_r}
      boxExtents=<38 38 38>
      constraintSetup {angularLimitsMax=<0 0 0>, angularLimitsMin=<0 0 0>, angularTarget=<1 0 0>, linearAxesMax=<0 0 1>, linearAxesMin=<0 0 -1>}
      localJointOffset=<2 0 0>
     }]
    relativeSpaceBone {boneName=pelvis}
    simSpaceSettings {maxAngularAcceleration=40, maxAngularVelocity=770, simSpaceAngularAlpha=0.4}
   }
   n17 = Anim Dynamics - Bone: upperarm_bicep_r(ComponentPose=@n16, ExternalForce="0,0,0") {
    alpha=0.8
    angularDampingOverride=0.9999
    angularSpringConstant=15000
    bAngularSpring=true
    bOverrideAngularDamping=true
    bUseGravityOverride=true
    bUsePlanarLimit=false
    simulationSpace=RootRelative
    boundBone {boneName=upperarm_bicep_r}
    componentAppliedLinearAccClamp=<1000 1000 1000>
    gravityOverride {z=-980}
    physicsBodyDefinitions [{
      boundBone {boneName=upperarm_bicep_r}
      boxExtents=<24 38 24>
      constraintSetup {angularLimitsMax=<1 10 1>, angularLimitsMin=<-1 -10 -1>, angularTarget=<1 0 0>}
      localJointOffset=<-5 0 0>
     }]
    relativeSpaceBone {boneName=pelvis}
    simSpaceSettings {maxAngularAcceleration=10, maxAngularVelocity=300, simSpaceAngularAlpha=0.4}
   }
   n18 = Anim Dynamics - Bone: upperarm_tricep_l!(ComponentPose=@n17, ExternalForce="0,0,0") -> (Pose) {
    alpha=0.8
    angularSpringConstant=15000
    bLinearSpring=true
    bOverrideLinearDamping=true
    bUseGravityOverride=true
    bUsePlanarLimit=false
    linearDampingOverride=0.9999
    linearSpringConstant=450
    simulationSpace=RootRelative
    boundBone {boneName=upperarm_tricep_l}
    componentAppliedLinearAccClamp=<1000 1000 1000>
    gravityOverride {z=-100}
    physicsBodyDefinitions [{
      boundBone {boneName=upperarm_tricep_l}
      boxExtents=<48 58 48>
      constraintSetup {angularLimitsMax=<0 0 0>, angularLimitsMin=<0 0 0>, angularTarget=<1 0 0>, linearAxesMax=<0 0.1 1.5>, linearAxesMin=<0 -0.1 -1.5>}
      localJointOffset=<-3 0 0>
     }]
    relativeSpaceBone {boneName=pelvis}
    simSpaceSettings {maxAngularAcceleration=30, maxAngularVelocity=300, simSpaceAngularAlpha=0.4}
   }
   n19 = Anim Dynamics - Bone: upperarm_tricep_l(ComponentPose=@n18, ExternalForce="0,0,0") {
    alpha=0.8
    angularDampingOverride=0.9999
    angularSpringConstant=50000
    bAngularSpring=true
    bOverrideAngularDamping=true
    bUseGravityOverride=true
    bUsePlanarLimit=false
    simulationSpace=RootRelative
    boundBone {boneName=upperarm_tricep_l}
    componentAppliedLinearAccClamp=<1000 1000 1000>
    gravityOverride {z=-980}
    physicsBodyDefinitions [{
      boundBone {boneName=upperarm_tricep_l}
      boxExtents=<24 38 24>
      constraintSetup {angularLimitsMax=<1 10 1>, angularLimitsMin=<-1 -10 -1>, angularTarget=<1 0 0>}
      localJointOffset=<5 0 0>
     }]
    relativeSpaceBone {boneName=pelvis}
    simSpaceSettings {maxAngularAcceleration=10, maxAngularVelocity=300, simSpaceAngularAlpha=0.4}
   }
   n20 = Anim Dynamics - Bone: upperarm_tricep_r!(ComponentPose=@n19, ExternalForce="0,0,0") -> (Pose) {
    alpha=0.8
    angularSpringConstant=15000
    bLinearSpring=true
    bOverrideLinearDamping=true
    bUseGravityOverride=true
    bUsePlanarLimit=false
    linearDampingOverride=0.9999
    linearSpringConstant=450
    simulationSpace=RootRelative
    boundBone {boneName=upperarm_tricep_r}
    componentAppliedLinearAccClamp=<1000 1000 1000>
    gravityOverride {z=-100}
    physicsBodyDefinitions [{
      boundBone {boneName=upperarm_tricep_r}
      boxExtents=<48 58 48>
      constraintSetup {angularLimitsMax=<0 0 0>, angularLimitsMin=<0 0 0>, angularTarget=<1 0 0>, linearAxesMax=<0 0.1 1.5>, linearAxesMin=<0 -0.1 -1.5>}
      localJointOffset=<3 0 0>
     }]
    relativeSpaceBone {boneName=pelvis}
    simSpaceSettings {maxAngularAcceleration=30, maxAngularVelocity=300, simSpaceAngularAlpha=0.4}
   }
   n21 = Anim Dynamics - Bone: upperarm_tricep_r(ComponentPose=@n20, ExternalForce="0,0,0") {
    alpha=0.8
    angularDampingOverride=0.9999
    angularSpringConstant=50000
    bAngularSpring=true
    bOverrideAngularDamping=true
    bUseGravityOverride=true
    bUsePlanarLimit=false
    simulationSpace=RootRelative
    boundBone {boneName=upperarm_tricep_r}
    componentAppliedLinearAccClamp=<1000 1000 1000>
    gravityOverride {z=-980}
    physicsBodyDefinitions [{
      boundBone {boneName=upperarm_tricep_r}
      boxExtents=<24 38 24>
      constraintSetup {angularLimitsMax=<1 10 1>, angularLimitsMin=<-1 -10 -1>, angularTarget=<1 0 0>}
      localJointOffset=<5 0 0>
     }]
    relativeSpaceBone {boneName=pelvis}
    simSpaceSettings {maxAngularAcceleration=10, maxAngularVelocity=300, simSpaceAngularAlpha=0.4}
   }
   n22 = Anim Dynamics - Bone: deform_pec_l!(ComponentPose=@n21, ExternalForce="0,0,0") -> (Pose) {
    alpha=0.8
    angularSpringConstant=15000
    bLinearSpring=true
    bOverrideLinearDamping=true
    bUseGravityOverride=true
    bUsePlanarLimit=false
    linearDampingOverride=0.9995
    linearSpringConstant=400
    simulationSpace=RootRelative
    boundBone {boneName=deform_pec_l}
    componentAppliedLinearAccClamp=<1000 1000 1000>
    gravityOverride {z=-200}
    physicsBodyDefinitions [{
      boundBone {boneName=deform_pec_l}
      boxExtents=<50 55 60>
      constraintSetup {angularLimitsMax=<0 0 0>, angularLimitsMin=<0 0 0>, angularTarget=<1 0 0>, linearAxesMax=<1 0 3>, linearAxesMin=<-1 0 -3>}
      localJointOffset=<-4 0 0>
     }]
    relativeSpaceBone {boneName=pelvis}
    simSpaceSettings {maxAngularAcceleration=10, maxAngularVelocity=770, simSpaceAngularAlpha=0.4}
   }
   n23 = Anim Dynamics - Bone: deform_pec_r!(ComponentPose=@n22, ExternalForce="0,0,0") -> (Pose) {
    alpha=0.8
    angularSpringConstant=15000
    bLinearSpring=true
    bOverrideLinearDamping=true
    bUseGravityOverride=true
    bUsePlanarLimit=false
    linearDampingOverride=0.9995
    linearSpringConstant=400
    simulationSpace=RootRelative
    boundBone {boneName=deform_pec_r}
    componentAppliedLinearAccClamp=<1000 1000 1000>
    gravityOverride {z=-200}
    physicsBodyDefinitions [{
      boundBone {boneName=deform_pec_r}
      boxExtents=<50 55 60>
      constraintSetup {angularLimitsMax=<0 0 0>, angularLimitsMin=<0 0 0>, angularTarget=<1 0 0>, linearAxesMax=<1 0 3>, linearAxesMin=<-1 0 -3>}
      localJointOffset=<4 0 0>
     }]
    relativeSpaceBone {boneName=pelvis}
    simSpaceSettings {maxAngularAcceleration=10, maxAngularVelocity=770, simSpaceAngularAlpha=0.4}
   }
   n24 = Component To Local(ComponentPose=@n23)
   n25 = Get bIsInFrontEnd()
   Save cached pose 'Dynamics_CACHE'(Pose=Blend Poses by bool(
     BlendPose_0=Blend Poses by bool(BlendPose_0=@n24, BlendPose_1=@n13, BlendTime_0=0, BlendTime_1=0, bActiveValue=@n25)
     BlendPose_1=Use cached pose 'PreDynamics_CACHE'()
     BlendTime_0=0
     BlendTime_1=0
     bActiveValue=Get bPbBlendEnabled() -> (bPbBlendEnabled:bool)
    )) {cacheName=Dynamics_CACHE}
  ]}
 EventGraph {nodes [Event BlueprintUpdateAnimation()]}
]
variables []
