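# Animation Blueprint for the DriftFable male medium face accessory, which makes its skeleton follow the character meshes it is worn with.
# The base pose and curves are copied from the attached parent mesh.
# A layered per-bone blend overlays shoulder, upper-arm and torso deform bones from the body mesh and facial bones from the head mesh.
# Both mesh references are fetched from the owning PlayerPawn_Athena when character customization finishes loading.
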AnimBlueprint /BRCosmetics/Characters/Player/Male/Medium/Bodies/M_MED_DriftFable/Meshes/Parts/M_MED_DriftFable_FaceAcc_AnimBP
target_skeleton=&/M_MED_DriftFable_FaceAcc_Skeleton
graphs [
 AnimGraph {nodes [
   Output Pose(Result=Use cached pose 'StartPose'()) {layerGroup=None}
   n1 = Layered blend per bone!(
    BasePose=Copy Pose From Mesh() {bCopyCurves=true, bUseAttachedParent=true}
    BlendPoses_0=Copy Pose From Mesh(SourceMeshComponent=Get Body_Mesh())
    BlendPoses_1:Pose Link Structure=Copy Pose From Mesh(SourceMeshComponent=Get Head_Mesh() -> (Head_Mesh:Skeletal Mesh Component Object Reference))
    BlendWeights_0
    BlendWeights_1:float=1
   ) -> (Pose) {
    blendPoses [{linkId=-1, sourceLinkId=-1}, {linkId=-1, sourceLinkId=-1}]
    blendWeights [1,1]
    layerSetup [
     {branchFilters [
       {boneName=deform_upperarm_up_l}
       {boneName=deform_upperarm_bk_l}
       {boneName=deform_upperarm_fr_l}
       {boneName=upperarm_twist_02_l}
       {boneName=deform_upperarm_up_r}
       {boneName=deform_upperarm_bk_r}
       {boneName=deform_upperarm_fr_r}
       {boneName=upperarm_twist_02_r}
       {boneName=deform_pec_l}
       {boneName=deform_pec_r}
       {boneName=deform_lat_l}
       {boneName=deform_lat_r}
       {boneName=clavicle_r}
       {boneName=upperarm_r}
       {boneName=clavicle_l}
       {boneName=upperarm_l}
      ]}
     {branchFilters [
       {boneName=C_brow_mid}
       {boneName=L_brow_mid}
       {boneName=L_brow_outer}
       {boneName=L_cheek_inner}
       {boneName=C_lip_upper_mid}
       {boneName=L_lip_upper_outer}
       {boneName=L_lip_corner}
       {boneName=C_nose_bridge}
       {boneName=R_brow_mid}
       {boneName=R_brow_outer}
       {boneName=R_cheek_inner}
       {boneName=R_lip_corner}
       {boneName=C_jaw}
       {boneName=L_lip_lower_outer}
       {boneName=C_lip_lower_mid}
       {boneName=R_lip_lower_outer}
       {boneName=R_lip_upper_outer}
      ]}
    ]
   }
   Save cached pose 'StartPose'(Pose=@n1) {cacheName=StartPose}
  ]}
 EventGraph {nodes [
   Event OnCharacterCustomizationLoaded()
   > n1 = Sequence()
   > n2 = Cast To PlayerPawn_Athena(Object=TryGetPawnOwner())
   > Set Head_Mesh(Head_Mesh:Skeletal Mesh Component Object Reference=GetSkeletalMeshForPartType(self=@n2, PartType=Head)) -> (Output_Get:Skeletal Mesh Component Object Reference)
   n3 = Cast To PlayerPawn_Athena(execute=@n1.then_1, Object=TryGetPawnOwner())
   > Set Body_Mesh(Body_Mesh=GetSkeletalMeshForPartType(self=@n3))
   Event BlueprintUpdateAnimation()
  ]}
]
variables [
 Body_Mesh {type=SkeletalMeshComponent}
 Head_Mesh {type=SkeletalMeshComponent}
]
